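# Niagara system for the EvilOrnament01 vehicle boost cosmetic, emitting blue trails and glow effects from skeletal mesh sockets, shifted by a per-car offset.
# The ribbon emitter spawns backward-moving trails per port that inherit source movement, swell then taper, and flicker secondary colors through floored sine waves.
# The sprite emitter partitions short-lived particles across ports, using two spawn groups to vary scale and offset for mesh and sprite renderers.
# Painted-variant override modules are disabled in both emitters.
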NiagaraSystem /VehicleCosmetics/Boosts/Boost_EvilOrnament01/FX/NS_Boost_EvilOrnament01
dependencies {
 usedDataInterfaces [&NiagaraDataInterfaceCurve, &NiagaraDataInterfaceSkeletalMesh]
 usedDynamicInputs [/VehicleCosmetics/FX/Niagara/Modules/MergeEmitters/VC_FloatBySpawnGroup]
 usedModules [
  &self:SystemSpawnScript.SystemScriptSource.SystemScriptGraph.NiagaraNodeAssignment_0.SetVariables_#1
  &SystemState
 ]
}
emitters {
 NE_VC_Boost_Ribbon_Parent {
  emitterUpdateScript [
   &EmitterState()
   &VC_FilteredSocketCount()
   &VC_RateQualityHandler(High Spawn Rate=60, Low Spawn Rate=30, Use Custom Rate?=true)
   &VC_SpawnRibbon_v1(Number of Ribbons=$Emitter.PortCount, SpawnRate=$Emitter.Spawn Rate)
  ]
  particleSpawnScript [
   &VC_PaintedVariantsOverride() {enabled=false}
   &VC_InitializeRibbon_v1(
    Custom Offset=true
    Enable Skeletal Position=true
    Local Vector Offset=&AddVectorToPosition(Position=<-35 0 0>, Vector=$User.DTCarOffset)
    Offset Amount=1
    Ribbon Width=28
   )
   &AddVelocity(Velocity=<-3000 0 0>) {module=AddVelocity001}
   &Color(Color=<rgba 0.1 0.175 1 1>)
  ]
  particleUpdateScript [
   &ParticleState()
   &InheritSourceMovement(
    Applied Position Scale=&Multiply_VectorByFloat(
     Vector=<1 1 1>
     Float=&FloatFromCurve(FloatCurve=DataInterface() {propertyValues {Curve {keys [{interpMode=RCIM_Cubic, tangentMode=RCTM_Break, time=0.5, value=1}, {interpMode=RCIM_Cubic, tangentMode=RCTM_Break, time=1, value=0}]}}})
    )
    ApplyForce=false
    ApplyPosition=true
    ApplyPositionFalloff=false
   )
   &SolveForcesAndVelocity(Clamp Velocity=false, Limit Acceleration=false)
   &VC_UpdateRibbon_v1(Enable Skeletal Position=true)
   &ScaleRibbonWidth(Ribbon Width Scale=&FloatFromCurve(FloatCurve=DataInterface() {propertyValues {Curve {keys [<at 0 0>,<at 0.1 1>,<at 0.5 2.5>,<at 1 0>]}}}))
   &ScaleColor(Scale RGB=&VectorFromFloat(Value=5), ScaleA=true, ScaleRGB=true)
   /SetVariables(
    Particles.RibbonColor2:LinearColor=/DealershipSharedLibrary/FX/Niagara/DynamicInputs/MultiplyLinearColorByVectorAndFloat(
     Color to Scale:LinearColor=$Particles.Color
     RGB Scale:Vector3f=&VectorFromFloat(Value=1)
     Alpha Scale:NiagaraFloat=&Multiply_Float(
      A=/Niagara/DynamicInputs/Math/Floor_Float(Float:NiagaraFloat=&Sine(Normalized Angle=$Emitter.Age, Period=0.3, Scale=5, Bias=0))
      B=&FloatFromCurve(FloatCurve=DataInterface() {propertyValues {Curve {keys [<at 0 1>,<at 0.8 0>]}}})
     )
    )
    Particles.RibbonColor3:LinearColor=&MultiplyLinearColorByVectorAndFloat(
     Color to Scale:LinearColor=$Particles.Color
     RGB Scale:Vector3f=&VectorFromFloat(Value=2)
     Alpha Scale:NiagaraFloat=&Multiply_Float(
      A=&Floor_Float(Float:NiagaraFloat=&Sine(Normalized Angle=$Emitter.Age, Period=0.3, Scale=5, Bias=0))
      B=&FloatFromCurve(FloatCurve=DataInterface() {propertyValues {Curve {keys [<at 0 1>,<at 0.8 0>]}}})
     )
    )
   ) {isSetParameters=true}
  ]
  rendererClasses [&NiagaraRibbonRendererProperties]
  renderers [
   {rendererClass=&NiagaraRibbonRendererProperties}
   {rendererClass=&NiagaraRibbonRendererProperties, rendererIndex=1}
   {rendererClass=&NiagaraRibbonRendererProperties, rendererIndex=2}
   {rendererClass=&NiagaraRibbonRendererProperties, rendererIndex=3}
  ]
 }
 NE_VC_Boost_Sprite_Parent {
  emitterUpdateScript [
   &EmitterState()
   &VC_FilteredSocketCount()
   &SpawnRate(SpawnRate=40, Use Spawn Probability=false)
   &SpawnRate(SpawnGroup=1, SpawnRate=40, Use Spawn Probability=false) {module=SpawnRate001}
  ]
  particleSpawnScript [
   &VC_PaintedVariantsOverride() {enabled=false}
   /SetVariables(Particles.VisibilityTag=&SpawnGroup()) {isSetParameters=true}
   &InitializeParticle(
    Color=<rgba 0.4 0.7 3 1>
    Lifetime=0.1
    Mesh Scale Max=/VehicleCosmetics/FX/Niagara/Modules/MergeEmitters/VC_VectorBySpawnGroup(
     SpawnGroup or ExecIndex:NiagaraInt32=0
     CustomInt:NiagaraInt32=&SpawnGroup()
     MaximumSpawnGroups:NiagaraInt32=2
     Vector Group 0:Vector3f=<0.65 0.65 0.65>
     Vector Group 1:Vector3f=<0.73 1 0.73>
     Vector Group 2:Vector3f=<0 0 0>
     Vector Group 3:Vector3f=<0 0 0>
     Vector Group 4:Vector3f=<0 0 0>
     Vector Group 5:Vector3f=<0 0 0>
    )
    Mesh Scale Min=&VC_VectorBySpawnGroup(
     SpawnGroup or ExecIndex:NiagaraInt32=0
     CustomInt:NiagaraInt32=&SpawnGroup()
     MaximumSpawnGroups:NiagaraInt32=2
     Vector Group 0:Vector3f=<0.45 0.45 0.45>
     Vector Group 1:Vector3f=<0.5 0.5 0.5>
     Vector Group 2:Vector3f=<0 0 0>
     Vector Group 3:Vector3f=<0 0 0>
     Vector Group 4:Vector3f=<0 0 0>
     Vector Group 5:Vector3f=<0 0 0>
    )
    Mesh Scale Mode=::NewEnumerator2
    Sprite Rotation Mode=::NewEnumerator2
    UsePositionOffset=false
    Write Material Random=true
   )
   &PartitionParticles(Number Of Partitions=$Emitter.PortCount, Partition Mode:ENiagara_PartitionMode=::NewEnumerator5)
   &SkeletalMeshLocation(
    Bone / Socket Index=$Particles.PartitionParticles.Partition
    Bone / Socket Sampling Mode=::NewEnumerator5
    Interpolate Mesh Sampling=false
    MaskSpawnGroup=false
    Mesh Sampling Type=::NewEnumerator6
    Sampled Position Offset=&AddVectorToPosition(
     Position=&VC_VectorBySpawnGroup(
      SpawnGroup or ExecIndex:NiagaraInt32=0
      CustomInt:NiagaraInt32=&SpawnGroup()
      MaximumSpawnGroups:NiagaraInt32=2
      Vector Group 0:Vector3f=<-60 0 0>
      Vector Group 1:Vector3f=<-42.5 0 0>
      Vector Group 2:Vector3f=<0 0 0>
      Vector Group 3:Vector3f=<0 0 0>
      Vector Group 4:Vector3f=<0 0 0>
      Vector Group 5:Vector3f=<0 0 0>
     )
     Vector=$User.DTCarOffset
    )
    Skeletal Mesh=$System.SkeletalMesh
   )
   &InitialMeshOrientation(
    Mesh Orientation Mode=::NewEnumerator7
    Rotation=&RandomRangeVector(Minimum=<-1 0 0>, Maximum=<1 0 0>)
    Rotation Coordinate Space=::NewEnumerator2
    Use Rotation Vector=true
   )
   &AddVelocity(Velocity=<-1500 0 0>) {enabled=false}
   /SetVariables(Particles.ColorTag1:LinearColor=<rgba 0.4 0.7 5 1>) {isSetParameters=true}
  ]
  particleUpdateScript [
   &ParticleState()
   &ScaleMeshSize(Scale Factor=&VC_VectorBySpawnGroup(
     SpawnGroup or ExecIndex:NiagaraInt32=0
     CustomInt:NiagaraInt32=&SpawnGroup()
     MaximumSpawnGroups:NiagaraInt32=2
     Vector Group 0:Vector3f=<0.79 0.79 0.79>
     Vector Group 1:Vector3f=<3.75 2.5 3.75>
     Vector Group 2:Vector3f=<0 0 0>
     Vector Group 3:Vector3f=<0 0 0>
     Vector Group 4:Vector3f=<0 0 0>
     Vector Group 5:Vector3f=<0 0 0>
    ))
   &SolveForcesAndVelocity(Clamp Velocity=false, Limit Acceleration=false)
   /SetVariables(Particles.MeshColor1:LinearColor=&MultiplyLinearColorByVectorAndFloat(
     Color to Scale:LinearColor=$Particles.Initial.ColorTag1
     RGB Scale:Vector3f=&VectorFromFloat(Value=4)
     Alpha Scale:NiagaraFloat=1
    )) {isSetParameters=true}
  ]
  rendererClasses [&NiagaraMeshRendererProperties, &NiagaraSpriteRendererProperties]
  renderers [
   {rendererClass=&NiagaraMeshRendererProperties, rendererIndex=0}
   {rendererClass=&NiagaraMeshRendererProperties, rendererIndex=1}
   {rendererClass=&NiagaraSpriteRendererProperties, rendererIndex=2}
  ]
 }
}
summary {
 systemName=NS_Boost_EvilOrnament01
 emitters [
  {
   emitterName=NE_VC_Boost_Sprite_Parent
   rendererClasses [&NiagaraMeshRendererProperties, &NiagaraSpriteRendererProperties]
  }
  {emitterName=NE_VC_Boost_Ribbon_Parent, rendererClasses [&NiagaraRibbonRendererProperties]}
 ]
 userVariables [
  User.bIsFrontendPreview {type=NiagaraBool}
  User.bIsPainted {type=NiagaraBool}
  User.ColorVariantID {type=NiagaraInt32}
  User.DTCarOffset {type=Vector3f, description="", defaultValue {struct=Vector3f, value=<0 0 0>}}
  User.PaintedColor {type=LinearColor, description="", defaultValue {struct=LinearColor, value=<rgba 1 1 1 1>}}
  User.PaintedScalar {type=NiagaraFloat, defaultValue=1, description=""}
  User.PortCount {type=NiagaraInt32, defaultValue=1}
 ]
}
